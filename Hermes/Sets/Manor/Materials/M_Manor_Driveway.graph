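# Manor driveway material, usable on spline meshes, rendering fully metallic gold bricks with darkened grout.
# Tiled gold albedo combines with rural-village brick normals and a contrast-adjusted mask; world-aligned grime varies color and roughness.
# Normals convert to world space with backface correction, then attributes pass through DirectFX, Veneer, jungle biome and RVT blending functions.
# Low-quality mobile overrides specular and roughness with constants before the Creative move-tool function.
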Material /Hermes/Sets/Manor/Materials/M_Manor_Driveway
graph {
 properties {bUseMaterialAttributes=true, bUsedWithSplineMeshes=true, usedShadingModels=Default Lit}
 nodes [
  n1 = Multiply(A=TextureCoordinate(), B=ScalarParameter() {defaultValue=1.5, parameterName=RoadTiling})
  n2 = TextureSample(UVs=@n1) {texture=/Hermes/Sets/Manor/Textures/T_Manor_Gold_D}
  n3 = Multiply(A=@n2.RGB, B=VectorParameter() {parameterName=Gold Color, defaultValue=<rgba 1.25 1.25 1.25 1>}.RGB)
  n4 = VectorParameter() {parameterName=Grout Color, defaultValue=<rgba 0 0 0 1>}
  n5 = Saturate(None=LinearInterpolate(
    A=ScalarParameter() {defaultValue=-0.8, parameterName=ContrastMmin}
    B=ScalarParameter() {defaultValue=1, parameterName=ContrastMax}
    Alpha=TextureSample(UVs=@n1) {samplerType=SAMPLERTYPE_Masks, texture=/Game/Environments/Helios/Sets/RuralVillage/Textures/T_RuralVillage_Brick_Ground_A_A_S}.B
   ))
  n6 = Multiply(A=@n5)
  n7 = VectorParameter() {parameterName=Grime Color, defaultValue=<rgba 0 0 0 1>}
  n8 = MaterialFunctionCall(
   TextureObject=TextureObject() {samplerType=SAMPLERTYPE_Masks, texture=/Game/Global/Textures/Default/Noise/T_Tiling_Grime}
   TextureSize=ScalarParameter() {defaultValue=512, parameterName=DirtScale}
   WorldPosition=WorldPosition().XYZ
  ) {materialFunction=&WorldAlignedTexture}
  n9 = MaterialFunctionCall(Float3=@n8.Z Texture) {materialFunction=&BreakOutFloat3Components}
  n10 = ScalarParameter() {defaultValue=0.4, parameterName=RoughnessMin}
  n11 = ScalarParameter() {defaultValue=0.2, parameterName=RoughnessMax}
  n12 = SetMaterialAttributes!(
   MaterialAttributes
   Base Color=LinearInterpolate(A=LinearInterpolate(A=@n3, B=@n4.RGB, Alpha=@n6), B=Multiply(A=@n2.RGB, B=@n7.RGB), Alpha=@n9.G)
   Normal=TextureSample(UVs=@n1) {samplerType=SAMPLERTYPE_Normal, texture=/Game/Environments/Helios/Sets/RuralVillage/Textures/T_RuralVillage_Brick_Ground_A_A_N}.RGB
   Specular
   Metallic=Constant() {r=1}
   Roughness=LinearInterpolate(
    A=LinearInterpolate(A=@n10, B=@n11, Alpha=@n5) {constA=1}
    B=ScalarParameter() {defaultValue=0.15, parameterName=Grime ROughness}
    Alpha=@n9.G
   )
  ) {attributeSetTypes [
    #1
    #2
    #3
    #4
    #5
   ]}
  Composite!(None=@n12) {desc=Transform to World in case we will use RVT after this point, subgraphName=Tangent to World}
  MaterialExpressionReroute_12 = Reroute(None=@n19.RVT H)
  MaterialExpressionReroute_3 = Reroute(None=@n12)
  n13 = GetMaterialAttributes(None=@MaterialExpressionReroute_3.Input) {attributeGetTypes [#2]}
  n14 = Transform(None=@n13.Normal)
  n15 = LinearInterpolate(A=Multiply(A=@n14) {constB=-1}, B=@n14, Alpha=Saturate(None=TwoSidedSign())) {desc="Fixes normals\r\non backfaces"}
  n16 = SetMaterialAttributes!(
   MaterialAttributes=@n13.MaterialAttributes
   Normal=StaticSwitchParameter(True=MaterialProxyReplace(Realtime=@n15, MaterialProxy=@n13.Normal), False=@n15) {parameterName=Force TangentSpaceNormalsFor HLOD ProxyReplace}
  ) {attributeSetTypes [#2]}
  MaterialExpressionReroute_2 = Reroute(None=@n16)
  n17 = MaterialFunctionCall!(In=MaterialFunctionCall!(In=@MaterialExpressionReroute_2.Output) {materialFunction=/Game/ArtTools/Tech_Art/DirectFX/Assets/MaterialFunctions/MF_DirectFX}.Result) {materialFunction=/Game/ArtTools/Tech_Art/Veneer/MaterialFunctions/MF_Veneer}
  MaterialExpressionReroute_11 = Reroute(None=@n17.Result)
  n18 = MaterialFunctionCall!(AO, RVT_Height=@MaterialExpressionReroute_12."RVT_Height (S)", For Buildings=@MaterialExpressionReroute_11."For Moss (MA)", For Moss) {materialFunction=/Game/Environments/Asteria/Rocks/Materials/MaterialFunctions/MF_JungleBiome}
  MaterialExpressionReroute_13 = Reroute(None=@n18.Buildings Jungl)
  n19 = MaterialFunctionCall!(
   In=@MaterialExpressionReroute_13.Result
   "RVT Height Map (Roughness)"=GetMaterialAttributes(None=@n17.Result) {attributeGetTypes [#5]}.Roughness
  ) {materialFunction=/Game/Global/Base4Layers_2022/MaterialFunctions/MF_RVTBlending}
  Composite!(InputPin=@n19.Result) {subgraphName=Mobile Opt}
  Composite!("For Moss (MA)"=@n17.Result, "RVT_Height (S)"=@n19.RVT H) {subgraphName="Seasonal "}
  MaterialExpressionReroute_1 = Reroute(None=@n19.Result)
  n20 = GetMaterialAttributes(None=@MaterialExpressionReroute_1.InputPin)
  n21 = Constant() {r=0.5}
  n22 = Constant() {r=0.8}
  n23 = ShadingPathSwitch(
   Default=@MaterialExpressionReroute_1.InputPin
   Mobile=QualitySwitch(
    Default=@MaterialExpressionReroute_1.InputPin
    Low=SetMaterialAttributes!(MaterialAttributes=@n20.MaterialAttributes, Specular=@n21, Roughness=@n22) {attributeSetTypes [#3, #5]}
    High=@MaterialExpressionReroute_1.InputPin
    Medium=@MaterialExpressionReroute_1.InputPin
    Epic=@MaterialExpressionReroute_1.InputPin
   )
  )
  MaterialExpressionReroute_4 = Reroute(None=@n23)
  MaterialFunctionCall(MA In=@MaterialExpressionReroute_4.Output) {materialFunction=&MF_CreativeMoveTool}
  n24 = TextureSample(UVs=@n1) {samplerType=SAMPLERTYPE_LinearGrayscale, texture=/Game/Environments/Helios/Sets/RuralVillage/Textures/T_RuralVillage_Brick_Ground_A_B_H}
  Multiply(A=@n24.B) {constB=0.7}
  NamedRerouteDeclaration(None=@n8.Z Texture) {name=Grime Mask, nodeColor=<rgba 0.2187 1 0 1>}
  PinBase!(Output=@n16) {reroutePins [Output {expression=&self:MaterialExpressionReroute_2}]}
  PinBase!() {reroutePins [Input {expression=&self:MaterialExpressionReroute_3}]}
  PinBase() {reroutePins [InputPin {expression=&self:MaterialExpressionReroute_1}]}
  PinBase!(Output=@n23) {reroutePins [Output {expression=&self:MaterialExpressionReroute_4}]}
  PinBase!() {reroutePins [
    "For Moss (MA)" {expression=&self:MaterialExpressionReroute_11}
    "RVT_Height (S)" {expression=&self:MaterialExpressionReroute_12}
   ]}
  PinBase!(Result=@n18.Buildings Jungl) {reroutePins [Result {expression=&self:MaterialExpressionReroute_13}]}
 ]
}
